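# Material function for Asteria grass foliage that computes an animated offset along a wind vector.
# The Noise input is shifted down by one half to center it, then multiplied by the OffsetDistance input and the OffsetAnimationDistance scalar parameter.
# That signed magnitude scales WindOffsetVector, producing back-and-forth displacement, which is output through the OffsetAnimation named reroute.
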MaterialFunction /Game/Environments/Asteria/Foliage/Grass/MF_VegOffsetAnimation
graph {nodes [
  n1 = FunctionInput() {bUsePreviewValueAsDefault=true, inputName=WindOffsetVector, previewValue=<1 0 0 1>}
  n2 = FunctionInput() {bUsePreviewValueAsDefault=true, inputName=Noise, inputType=FunctionInput_Scalar, previewValue=<0.5 0 0 1>}
  n3 = Constant() {r=0.5}
  n4 = Multiply(
   A=Multiply(
    A=Add(A=@n2) {constB=-0.5}
    B=FunctionInput(Preview=@n3) {bUsePreviewValueAsDefault=true, inputName=OffsetDistance, inputType=FunctionInput_Scalar}
   )
   B=ScalarParameter() {defaultValue=300, parameterName=OffsetAnimationDistance}
  )
  FunctionOutput(None=NamedRerouteDeclaration(None=Multiply(A=@n1, B=@n4)) {name=OffsetAnimation, nodeColor=<rgba 0 1 0.8359 1>})
 ]}
